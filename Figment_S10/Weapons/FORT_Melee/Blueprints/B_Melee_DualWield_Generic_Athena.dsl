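; Generic dual-wield melee weapon Blueprint for Athena that adds a left-hand weapon mesh and drives its cosmetic effects.
; On equip it attaches the left mesh to the pawn and binds the pawn's left/right swing dispatchers to trail and swing FX.
; Construction and BeginPlay configure each hand's particles, with optional left-side overrides.
; It also handles force feedback, impact FX repositioning, cosmetic alteration particles and a durability-destroy effect.
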(event Custom|Swing%20Left%202
  (bind _left_swingfx (Variables|Default|GetLeft_SwingFX))
  (Effects|Particles|Trails|BeginTrails (Variables|Default|GetLeft_AnimTrail) (select (Variables|Effects|GetUse%20Different%20Effects%20for%20Left%20and%20Right) (Variables|Effects|GetLeft_FirstSocketName) (Variables|Effects|GetFirstSocketName)) (select (Variables|Effects|GetUse%20Different%20Effects%20for%20Left%20and%20Right) (Variables|Effects|GetLeft_Second%20Socket%20Name) (Variables|Effects|GetSecond%20Socket%20Name)) "ETrailWidthMode_FromCentre" (select (Variables|Effects|GetUse%20Different%20Effects%20for%20Left%20and%20Right) (Variables|Effects|GetLeft_Width) (Variables|Effects|GetWidth)))
  (Utilities|IsValid _left_swingfx
    (:Is%20Valid
      (Components|Activation|Deactivate _left_swingfx))
    (:Is%20Not%20Valid
      _))
  (Components|Activation|Activate _left_swingfx false))

(event Custom|Swing%20Left
  (bind _left_swingfx (Variables|Default|GetLeft_SwingFX))
  (Effects|Particles|Trails|BeginTrails (Variables|Default|GetLeft_AnimTrail) (select (Variables|Effects|GetUse%20Different%20Effects%20for%20Left%20and%20Right) (Variables|Effects|GetLeft_FirstSocketName) (Variables|Effects|GetFirstSocketName)) (select (Variables|Effects|GetUse%20Different%20Effects%20for%20Left%20and%20Right) (Variables|Effects|GetLeft_Second%20Socket%20Name) (Variables|Effects|GetSecond%20Socket%20Name)) "ETrailWidthMode_FromCentre" (select (Variables|Effects|GetUse%20Different%20Effects%20for%20Left%20and%20Right) (Variables|Effects|GetLeft_Width) (Variables|Effects|GetWidth)))
  (Utilities|IsValid _left_swingfx
    (:Is%20Valid
      (Components|Activation|Deactivate _left_swingfx))
    (:Is%20Not%20Valid
      _))
  (Components|Activation|Activate _left_swingfx false))

(event Custom|Swing%20Right%202
  (bind _swingfx (Variables|Default|GetSwingFX))
  (Effects|Particles|Trails|BeginTrails (Variables|Default|GetAnimTrail) (Variables|Effects|GetFirstSocketName) (Variables|Effects|GetSecond%20Socket%20Name) "ETrailWidthMode_FromCentre" (Variables|Effects|GetWidth))
  (Utilities|IsValid _swingfx
    (:Is%20Valid
      (Components|Activation|Deactivate _swingfx))
    (:Is%20Not%20Valid
      _))
  (Components|Activation|Activate _swingfx false))

(event Custom|Swing%20Right
  (bind _swingfx (Variables|Default|GetSwingFX))
  (Effects|Particles|Trails|BeginTrails (Variables|Default|GetAnimTrail) (Variables|Effects|GetFirstSocketName) (Variables|Effects|GetSecond%20Socket%20Name) "ETrailWidthMode_FromCentre" (Variables|Effects|GetWidth))
  (Utilities|IsValid _swingfx
    (:Is%20Valid
      (Components|Activation|Deactivate _swingfx))
    (:Is%20Not%20Valid
      _))
  (Components|Activation|Activate _swingfx false))

(event EventBeginPlay
  (bind _left_swingfx (Variables|Default|GetLeft_SwingFX))
  (bind _swingfx (Variables|Default|GetSwingFX))
  (bind _idlefx (Variables|Default|GetIdleFX))
  (bind _left_idlefx (Variables|Default|GetLeft_IdleFX))
  ;[1 1 1] Event Begin Play
  (Utilities|IsValid _idlefx
    (:Is%20Valid
      (Effects|Components|ParticleSystem|SetAutoAttachmentParameters _idlefx (Variables|Visual|GetWeaponMesh) (Variables|Effects|GetIdleFXAttachSocket) (Variables|Effects|GetIdleFX%20Location%20Rule) (Variables|Effects|GetIdle%20FX%20Rotation%20Rule) (Variables|Effects|GetIdle%20FX%20Scale%20Rule)))
    (:Is%20Not%20Valid
      _))
  (Utilities|IsValid _swingfx
    (:Is%20Valid
      (Effects|Components|ParticleSystem|SetAutoAttachmentParameters _swingfx (Variables|Visual|GetWeaponMesh) (Variables|Effects|GetSwingFXSocket) (Variables|Effects|GetSwingFX%20Location%20Rule) (Variables|Effects|GetSwing%20FX%20Rotation%20Rule) (Variables|Effects|GetSwing%20FX%20Scale%20Rule))
      (Utilities|IsValid _left_idlefx
        (:Is%20Valid
          (Effects|Components|ParticleSystem|SetAutoAttachmentParameters _left_idlefx (Variables|Default|GetLeft_WeaponMesh) (select (Variables|Effects|GetUse%20Different%20Effects%20for%20Left%20and%20Right) (Variables|Effects|GetLeft_IdleFXAttachSocket) (Variables|Effects|GetIdleFXAttachSocket)) (select (Variables|Effects|GetUse%20Different%20Effects%20for%20Left%20and%20Right) (Variables|Effects|GetLeft_IdleFX%20Location%20Rule) (Variables|Effects|GetIdleFX%20Location%20Rule)) (select (Variables|Effects|GetUse%20Different%20Effects%20for%20Left%20and%20Right) (Variables|Effects|GetLeft_Idle%20FX%20Rotation%20Rule) (Variables|Effects|GetIdle%20FX%20Rotation%20Rule)) (select (Variables|Effects|GetUse%20Different%20Effects%20for%20Left%20and%20Right) (Variables|Effects|GetLeft_Idle%20FX%20Scale%20Rule) (Variables|Effects|GetIdle%20FX%20Scale%20Rule))))
        (:Is%20Not%20Valid
          _))
      (Utilities|IsValid _left_swingfx
        (:Is%20Valid
          (Effects|Components|ParticleSystem|SetAutoAttachmentParameters _left_swingfx (Variables|Default|GetLeft_WeaponMesh) (select (Variables|Effects|GetUse%20Different%20Effects%20for%20Left%20and%20Right) (Variables|Effects|GetLeft_SwingFXSocket) (Variables|Effects|GetSwingFXSocket)) (select (Variables|Effects|GetUse%20Different%20Effects%20for%20Left%20and%20Right) (Variables|Effects|GetLeft_SwingFX%20Location%20Rule) (Variables|Effects|GetSwingFX%20Location%20Rule)) (select (Variables|Effects|GetUse%20Different%20Effects%20for%20Left%20and%20Right) (Variables|Effects|GetLeft_Swing%20FX%20Rotation%20Rule) (Variables|Effects|GetSwing%20FX%20Rotation%20Rule)) (select (Variables|Effects|GetUse%20Different%20Effects%20for%20Left%20and%20Right) (Variables|Effects|GetLeft_Swing%20FX%20Scale%20Rule) (Variables|Effects|GetSwing%20FX%20Scale%20Rule))))))))

(event Custom|Swing%20Left%20End
  (Effects|Particles|Trails|EndTrails (Variables|Default|GetLeft_AnimTrail)))

(event Custom|Swing%20Right%20End
  (Effects|Particles|Trails|EndTrails (Variables|Default|GetAnimTrail)))

(event FortWeapon|EventOnInstigatorSet
  ;[1 1 1] Event Equip and UnEquip
  (if (Variables|Default|GetbEquipPendingInstigator)
    (Variables|Default|SetbEquipPendingInstigator false)
    (bind _asfort_pawn (Utilities|Casting|CastToFortPawn (Game|GetInstigator))
      (:then
        (Utilities|IsValid (Variables|/Script/Engine.Character|GetMesh _asfort_pawn)
          (:Is%20Valid
            _)
          (:Is%20Not%20Valid
            (Utilities|FlowControl|Delay 0.1)))
        (Transformation|AttachComponentToComponent (Variables|Default|GetLeft_WeaponMesh) (Variables|/Script/Engine.Character|GetMesh _asfort_pawn) (Variables|Default|GetLeft%20Weapon%20Socket%20Name) "SnapToTarget" "SnapToTarget" "SnapToTarget" true)
        (CallFunction|SetActiveAlterationIdleParticles true)
        (CallFunction|BindDualMeleeSwingEvents))
      (:CastFailed
        _))
    (Variables|Default|SetbEquipPendingInstigator true)))

(event FortWeapon|EventOnEquippedAndAttached
  ;[1 1 1] Event Equip and UnEquip
  (bind _asfort_pawn (Utilities|Casting|CastToFortPawn (Game|GetInstigator))
    (:then
      (Utilities|IsValid (Variables|/Script/Engine.Character|GetMesh _asfort_pawn)
        (:Is%20Valid
          _)
        (:Is%20Not%20Valid
          (Utilities|FlowControl|Delay 0.1)))
      (Transformation|AttachComponentToComponent (Variables|Default|GetLeft_WeaponMesh) (Variables|/Script/Engine.Character|GetMesh _asfort_pawn) (Variables|Default|GetLeft%20Weapon%20Socket%20Name) "SnapToTarget" "SnapToTarget" "SnapToTarget" true)
      (CallFunction|SetActiveAlterationIdleParticles true)
      (CallFunction|BindDualMeleeSwingEvents))
    (:CastFailed
      _))
  (Variables|Default|SetbEquipPendingInstigator true))

(event FortWeapon|EventOnLocalUnEquip
  ;[1 1 1] Event Equip and UnEquip
  (Effects|Particles|Trails|EndTrails (Variables|Default|GetAnimTrail))
  (Transformation|DetachFromComponent (Variables|Default|GetLeft_WeaponMesh) "KeepRelative" "KeepRelative" "KeepRelative" true)
  (CallFunction|SetActiveAlterationIdleParticles false)
  (CallFunction|UnbindDualMeleeSwingEvents))

(event EventOnPlayWeaponFireFX
  (params (bPersistentFire :type Boolean) (bSecondaryFire :type Boolean))
  ;[1 1 1] Event on Play Weapon Fire FX - called from Combat Melee Montage AnimNotify PlayFireFX
  (FortWeapon|PlayForceFeedback false))

(event Custom|PlayRClickImpacts
  ;[1 1 1] Spawn RClick Impacts from AnimNotify
  (Components|Activation|Activate (Variables|Default|GetMeleeHeavy_PSC) false))

(event EventOnInitCosmeticAlterations
  (params (CosmeticMod :type Fort%20Cosmetic%20Modification%20Structure))
  (bind (_cosmeticmaterial _ambientparticlesystem) (Utilities|Struct|BreakFortCosmeticModification CosmeticMod))
  ;[1 1 1] Cosmetic Alterations
  ;; Elemental Alteration
  (Networking|CanExecuteCosmeticEvents
    (:True
      (Utilities|FlowControl|Sequence
        (:then_0
          (if (not (Asset|IsAssetNull _ambientparticlesystem))
            (bind _object (Utilities|AsyncLoadAsset _ambientparticlesystem)
              (:Completed
                (Utilities|Casting|CastToParticleSystem _object)
                (Utilities|ElementalWeaponSetupFX (Variables|Visual|GetWeaponMesh) 0 "FX_Idle_Start" "FX_Idle_End" false
                  (:Out
                    (bind _output_get (Variables|Default|SetAlteration%20Ambient%20PS 0))
                    (LOD|SetMaxDrawDistance _output_get 5e+03)))))))
        (:then_1
          (bind _asplayer_pawn_generic_parent (Utilities|Casting|CastToPlayerPawn_Generic_Parent (Game|GetInstigator self)))
          (Class|PlayerPawnGenericParent|ReinitializeWeaponMaterials _asplayer_pawn_generic_parent))))))

(event FortWeapon|EventOnPlayImpactFX
  (params (HitResult :type Hit%20Result%20Structure%20%28by%20ref%29) (ImpactPhysicalSurface :type EPhysicalSurface%20Enum :default "SurfaceType_Default") (SpawnedPSC :type FXSystem%20Component%20Object%20Reference))
  (bind (_bblockinghit _binitialoverlap _time _distance _location _impactpoint _normal _impactnormal _physmat _hitactor) (Collision|BreakHitResult HitResult))
  (bind _returnvalue (Transformation|GetActorLocation _hitactor))
  ;[1 1 1] Impact FX
  (|CallParentFunction HitResult ImpactPhysicalSurface SpawnedPSC)
  (FortWeapon|PlayForceFeedback (Math|Boolean|ORBoolean (Utilities|Enum|Equal(Enum) ImpactPhysicalSurface "SurfaceType14") (Utilities|Enum|Equal(Enum) ImpactPhysicalSurface "SurfaceType15") (Utilities|Enum|Equal(Enum) ImpactPhysicalSurface "SurfaceType16")) false true)
  (if (Math|Boolean|ORBoolean (Utilities|Enum|Equal(Enum) ImpactPhysicalSurface "SurfaceType6") (Utilities|Enum|Equal(Enum) ImpactPhysicalSurface "SurfaceType7") (Utilities|Enum|Equal(Enum) ImpactPhysicalSurface "SurfaceType4"))
    (bind _asfort_pawn (Utilities|Casting|CastToFortPawn (Game|GetInstigator _hitactor)))
    (Utilities|IsValid (Components|SkeletalMesh|GetAnimInstance (Variables|/Script/Engine.Character|GetMesh _asfort_pawn))
      (:Is%20Valid
        (if (Utilities|Array|ContainsItem (Variables|Effects|GetPokeAnimations) (Animation|Montage|GetCurrentActiveMontage (Components|SkeletalMesh|GetAnimInstance (Variables|/Script/Engine.Character|GetMesh _asfort_pawn))))
          (Effects|Components|ParticleSystem|SetEmitterEnable SpawnedPSC "Sparks" false)
          (Effects|Components|ParticleSystem|SetEmitterEnable SpawnedPSC "Swoosh" false)
          (else
            _))
        (Transformation|SetWorldLocationAndRotation SpawnedPSC (select (Utilities|Enum|Equal(Enum) ImpactPhysicalSurface "SurfaceType4") (Math|Vector|Lerp(Vector) (Transformation|GetSocketLocation (Variables|/Script/Engine.Character|GetMesh _asfort_pawn) "FX_Chest") (Transformation|GetSocketLocation (Variables|Visual|GetWeaponMesh) "SwingFX") 0.25) (+ (Transformation|GetSocketLocation (Variables|/Script/Engine.Character|GetMesh _asfort_pawn) "FX_Chest") (* (Math|Vector|Normalize (Math|Vector|MakeVector (.x (- _returnvalue (Transformation|GetSocketLocation (Variables|/Script/Engine.Character|GetMesh _asfort_pawn) "FX_Chest"))) (.y (- _returnvalue (Transformation|GetSocketLocation (Variables|/Script/Engine.Character|GetMesh _asfort_pawn) "FX_Chest"))) (.z (- _returnvalue (Transformation|GetSocketLocation (Variables|/Script/Engine.Character|GetMesh _asfort_pawn) "FX_Chest"))))) (Math|Vector|VectorLength (- _location (Transformation|GetSocketLocation (Variables|/Script/Engine.Character|GetMesh _asfort_pawn) "FX_Chest")))))) (Math|Rotator|MakeRotfromXY (Math|Vector|CrossProduct (Math|Vector|CrossProduct (Math|Vector|Normalize (- (Transformation|GetSocketLocation (Variables|Visual|GetWeaponMesh) "SwingFX") (Transformation|GetSocketLocation (Variables|/Script/Engine.Character|GetMesh _asfort_pawn) "FX_Chest"))) (Math|Rotator|GetRotationXVector (Transformation|GetSocketRotation (Variables|Visual|GetWeaponMesh) "SwingFX"))) (Math|Vector|Normalize (Math|Rotator|GetRotationXVector (Transformation|GetSocketRotation (Variables|/Script/Engine.Character|GetMesh _asfort_pawn) "FX_Combat")))) (Math|Rotator|GetRotationXVector (Transformation|GetSocketRotation (Variables|/Script/Engine.Character|GetMesh _asfort_pawn) "FX_Combat"))) false false)
        (Effects|Components|ParticleSystem|SetVectorParameter SpawnedPSC "PoofLocation" (- _location (+ (Transformation|GetSocketLocation (Variables|/Script/Engine.Character|GetMesh _asfort_pawn) "FX_Chest") (* (Math|Vector|Normalize (Math|Vector|MakeVector (.x (- _returnvalue (Transformation|GetSocketLocation (Variables|/Script/Engine.Character|GetMesh _asfort_pawn) "FX_Chest"))) (.y (- _returnvalue (Transformation|GetSocketLocation (Variables|/Script/Engine.Character|GetMesh _asfort_pawn) "FX_Chest"))) (.z (- _returnvalue (Transformation|GetSocketLocation (Variables|/Script/Engine.Character|GetMesh _asfort_pawn) "FX_Chest"))))) (Math|Vector|VectorLength (- _location (Transformation|GetSocketLocation (Variables|/Script/Engine.Character|GetMesh _asfort_pawn) "FX_Chest"))))))))
      (:Is%20Not%20Valid
        _))))

(event EventOnEquippedWeaponDestory
  ;[0.725 0.6854 0.4277] Show fx for weapon destruction on depleted durability
  (if (Variables|Effects|GetUseDestroyEffect)
    (Effects|SpawnEmitteratLocation (Variables|Effects|GetWeaponDurabilityDestroyEffect) (Transformation|GetSocketLocation (Variables|Visual|GetWeaponMesh) "Melee_TwoHanded_Handle"))
    (bind _asfort_pawn (Utilities|Casting|CastToFortPawn (Game|GetInstigator)))
    (bind _returnvalue (Effects|SpawnEmitterAttached :EmitterTemplate (Variables|Effects|GetWeaponDurabilityDestroyEffectIcon) :AttachToComponent (Variables|/Script/Engine.Character|GetMesh _asfort_pawn) :AttachPointName "Root_Effect" :Location "0, 0,130.000000" :Scale "1.000000,1.000000,1.000000" :LocationType "SnapToTarget" :bAutoDestroy true :PoolingMethod "None" :bAutoActivate true))
    (bind _returnvalue_2 (Rendering|Material|CreateDynamicMaterialInstance _returnvalue 0 "/Game/Effects/WeaponDestroy/M_WeaponDestroyExclamation.M_WeaponDestroyExclamation" "None" :!class /Script/Engine.PrimitiveComponent))
    (Rendering|Material|SetMaterial _returnvalue 0 _returnvalue_2)
    (Rendering|Material|SetTextureParameterValue _returnvalue_2 "IconTex" (ItemPreviewImage|GetorLoadSynchronously (Item|GetSmallPreviewImage (Data|GetWeaponData))))))

(event EventOnWeaponVisibilityChanged
  (params (bVisible :type Boolean) (bSetForLocalControllerOnly :type Boolean))
  (bind _alteration_ambient_ps (Variables|Default|GetAlteration%20Ambient%20PS))
  ;[1 1 1] EVENT -> OnWeaponVisibilityChanged
  (Utilities|IsValid (Variables|Default|GetAlteration%20Ambient%20PS)
    (:Is%20Valid
      (if bVisible
        (Components|Activation|Activate _alteration_ambient_ps true)
        (else
          (Components|Activation|Deactivate _alteration_ambient_ps))))))

(fn BindDualMeleeSwingEvents
  (bind _asplayer_pawn_athena (Utilities|Casting|CastToPlayerPawn_Athena (Game|GetInstigator))
    (:then
      (Default|BindEventtoSwingLeft _asplayer_pawn_athena (EventDispatchers|CreateEvent :!event Swing%20Left))
      (Default|BindEventtoSwingLeftEnd _asplayer_pawn_athena (EventDispatchers|CreateEvent :!event Swing%20Left%20End))
      (Default|BindEventtoSwingRight _asplayer_pawn_athena (EventDispatchers|CreateEvent :!event Swing%20Right))
      (Default|BindEventtoSwingRightEnd _asplayer_pawn_athena (EventDispatchers|CreateEvent :!event Swing%20Right%20End))
      (Default|BindEventtoSwingLeft2 _asplayer_pawn_athena (EventDispatchers|CreateEvent :!event Swing%20Left%202))
      (Default|BindEventtoSwingRight2 _asplayer_pawn_athena (EventDispatchers|CreateEvent :!event Swing%20Right%202)))
    (:CastFailed
      (bind _asplayer_pawn_generic (Utilities|Casting|CastToPlayerPawn_Generic (Game|GetInstigator))
        (:then
          (Default|BindEventtoSwingLeft _asplayer_pawn_generic (EventDispatchers|CreateEvent :!event Swing%20Left))
          (Default|BindEventtoSwingLeftEnd _asplayer_pawn_generic (EventDispatchers|CreateEvent :!event Swing%20Left%20End))
          (Default|BindEventtoSwingRight _asplayer_pawn_generic (EventDispatchers|CreateEvent :!event Swing%20Right))
          (Default|BindEventtoSwingRightEnd _asplayer_pawn_generic (EventDispatchers|CreateEvent :!event Swing%20Right%20End))
          (Default|BindEventtoSwingLeft2 _asplayer_pawn_generic (EventDispatchers|CreateEvent :!event Swing%20Left%202))
          (Default|BindEventtoSwingRight2 _asplayer_pawn_generic (EventDispatchers|CreateEvent :!event Swing%20Right%202)))
        (:CastFailed
          (Development|PrintString (Utilities|String|Append "Cannot bind events.  Cast Failed!  Instigator of dual-wield melee weapon is not a Player Pawn Athena OR PlayerPawn_Generic (the StW Pawn), so it will not dual-wield properly. Instigator: " (Utilities|String|ToString(Object) (Game|GetInstigator)))) ; Instigator class is not supported!
        )))))

(fn GetInWater
  (return false))

(fn Melee_Effect_Color
  (return (Variables|Effects|GetEffects_Color_Level)))

(fn SetActiveAlterationIdleParticles
  (params (Active :type Boolean) (Reset :type Boolean))
  (Networking|CanExecuteCosmeticEvents
    (:True
      (bind _alteration_ambient_ps (Variables|Default|GetAlteration%20Ambient%20PS)
        (:then
          (Components|Activation|SetActive _alteration_ambient_ps Active Reset)
          (return))
        (:else)))))

(fn SetWpnRarity
  ;[1 1 1] Wpn Rarity
  (Development|PrintString (Utilities|String|ToString(Float) (Math|Conversions|ToFloat(Byte) (Utilities|Select :Index (Item|GetRarity (Data|GetWeaponData)) :Common 0 :Uncommon 1 :Rare 2 :Epic 3 :Legendary 4 :Mythic 5 :Transcendent 6 :Unattainable 7)))))

(fn UnbindDualMeleeSwingEvents
  (bind _asplayer_pawn_athena (Utilities|Casting|CastToPlayerPawn_Athena (Game|GetInstigator))
    (:then
      (Default|UnbindEventfromSwingLeft _asplayer_pawn_athena (EventDispatchers|CreateEvent :!event Swing%20Left))
      (Default|UnbindEventfromSwingLeftEnd _asplayer_pawn_athena (EventDispatchers|CreateEvent :!event Swing%20Left%20End))
      (Default|UnbindEventfromSwingRight _asplayer_pawn_athena (EventDispatchers|CreateEvent :!event Swing%20Right))
      (Default|UnbindEventfromSwingRightEnd _asplayer_pawn_athena (EventDispatchers|CreateEvent :!event Swing%20Right%20End))
      (Default|UnbindEventfromSwingLeft2 _asplayer_pawn_athena (EventDispatchers|CreateEvent :!event Swing%20Left%202))
      (Default|UnbindEventfromSwingRight2 _asplayer_pawn_athena (EventDispatchers|CreateEvent :!event Swing%20Right%202)))
    (:CastFailed
      (bind _asplayer_pawn_generic (Utilities|Casting|CastToPlayerPawn_Generic (Game|GetInstigator))
        (:then
          (Default|UnbindEventfromSwingLeft _asplayer_pawn_generic (EventDispatchers|CreateEvent :!event Swing%20Left))
          (Default|UnbindEventfromSwingLeftEnd _asplayer_pawn_generic (EventDispatchers|CreateEvent :!event Swing%20Left%20End))
          (Default|UnbindEventfromSwingRight _asplayer_pawn_generic (EventDispatchers|CreateEvent :!event Swing%20Right))
          (Default|UnbindEventfromSwingRightEnd _asplayer_pawn_generic (EventDispatchers|CreateEvent :!event Swing%20Right%20End))
          (Default|UnbindEventfromSwingLeft2 _asplayer_pawn_generic (EventDispatchers|CreateEvent :!event Swing%20Left%202))
          (Default|UnbindEventfromSwingRight2 _asplayer_pawn_generic (EventDispatchers|CreateEvent :!event Swing%20Right%202)))
        (:CastFailed
          (Development|PrintString (Utilities|String|Append "Cannot unbind events.  Cast Failed!  Instigator of dual-wield melee weapon is not a Player Pawn Athena OR PlayerPawn_Generic (the StW Pawn), so it will not dual-wield properly.  Instigator: " (Utilities|String|ToString(Object) (Game|GetInstigator)))) ; Instigator class is not supported!
        )))))

(fn ConstructionScript
  ;[1 1 1] SetupMeleeFXbyTag AnimNotify_PlayFX_WeapType uses this info
  (Effects|Components|ParticleSystem|SetTemplate (Variables|Default|GetMeleeHeavy_PSC) (Variables|Effects|GetMeleeHeavy_ParticleSystem))
  (Effects|Components|ParticleSystem|SetTemplate (Variables|Default|GetSwingFX) (Variables|Effects|GetSwing%20FX%20Template))
  (Effects|Components|ParticleSystem|SetTemplate (Variables|Default|GetLeft_SwingFX) (select (Variables|Effects|GetUse%20Different%20Effects%20for%20Left%20and%20Right) (Variables|Effects|GetLeft_Swing%20FX%20Template) (Variables|Effects|GetSwing%20FX%20Template)))
  (Effects|Components|ParticleSystem|SetTemplate (Variables|Default|GetIdleFX) (Variables|Effects|GetIdleFX%20Template))
  (Effects|Components|ParticleSystem|SetTemplate (Variables|Default|GetLeft_IdleFX) (select (Variables|Effects|GetUse%20Different%20Effects%20for%20Left%20and%20Right) (Variables|Effects|GetLeft_IdleFX%20Template) (Variables|Effects|GetIdleFX%20Template)))
  (Effects|Components|ParticleSystem|SetTemplate (Variables|Default|GetAnimTrail) (Variables|Effects|GetAnim%20Trail%20Template))
  (Effects|Components|ParticleSystem|SetTemplate (Variables|Default|GetLeft_AnimTrail) (select (Variables|Effects|GetUse%20Different%20Effects%20for%20Left%20and%20Right) (Variables|Effects|GetLeft_Anim%20Trail%20Template) (Variables|Effects|GetAnim%20Trail%20Template))))
