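# Physics asset for the Figment S10 two-handed sledgehammer melee weapon skeletal mesh.
# A single box body on the Melee_TwoHanded_Handle bone provides the weapon's collision shape, sized and offset to cover the hammer.
# It has no constraints, so the hammer simulates as one rigid piece with no articulated parts.
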PhysicsAsset /Figment_S10/Weapons/FORT_Melee/Meshes/SK_SledgeHammer_Physics
bodies {Melee_TwoHanded_Handle [{shapeType=Box, extentX=38.2365, extentY=13.1648, extentZ=112.8633, center=<0.4465 0.0713 21.79>, rotation=<rot -0.0571 -90 -0.137>}]}
constraints []
